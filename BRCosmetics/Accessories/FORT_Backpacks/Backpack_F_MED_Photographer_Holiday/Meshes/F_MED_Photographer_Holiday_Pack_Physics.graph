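# Physics asset for the holiday Photographer backpack mesh, defining the simulation setup for its left and right wings.
# Each side chains a small capsule on the wing root bone to three dynamic capsule bodies down the wing.
# Constraints link each segment to its parent with narrow swing limits, tighter at the root joint, and zero twist limit, allowing only slight bending.
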PhysicsAsset /BRCosmetics/Accessories/FORT_Backpacks/Backpack_F_MED_Photographer_Holiday/Meshes/F_MED_Photographer_Holiday_Pack_Physics
bodies {
 L_wing [{shapeType=Capsule, length=0, radius=0.6782, shapeName=L_wing_capsule, center=<1.7018 0 0>, rotation=<rot 90 0 0>}]
 R_wing [{shapeType=Capsule, length=0, radius=0.6782, shapeName=R_wing_capsule, center=<1.7018 0 0>, rotation=<rot 90 0 0>}]
 dyn_L_wing_1 [{shapeType=Capsule, length=3.2422, radius=2, shapeName=dyn_L_wing_1_capsule, center=<4 0 0>, rotation=<rot 90 0 0>}]
 dyn_L_wing_2 [{shapeType=Capsule, length=2, radius=2, shapeName=dyn_L_wing_2_capsule, center=<3 0 0>, rotation=<rot 90 0 0>}]
 dyn_L_wing_3 [{shapeType=Capsule, length=2, radius=2, shapeName=dyn_L_wing_3_capsule, center=<3 0 0>, rotation=<rot 90 0 0>}]
 dyn_R_wing_1 [{shapeType=Capsule, length=3.2422, radius=2, shapeName=dyn_R_wing_1_capsule, center=<4 0 0>, rotation=<rot 90 0 0>}]
 dyn_R_wing_2 [{shapeType=Capsule, length=2, radius=2, shapeName=dyn_R_wing_2_capsule, center=<3 0 0>, rotation=<rot 90 0 0>}]
 dyn_R_wing_3 [{shapeType=Capsule, length=2, radius=2, shapeName=dyn_R_wing_3_capsule, center=<3 0 0>, rotation=<rot 90 0 0>}]
}
constraints [
 {bone1Name=dyn_L_wing_2, bone2Name=dyn_L_wing_1, swing1LimitDegrees=7, swing2LimitDegrees=7, twistLimitDegrees=0}
 {bone1Name=dyn_L_wing_3, bone2Name=dyn_L_wing_2, swing1LimitDegrees=7, swing2LimitDegrees=7, twistLimitDegrees=0}
 {bone1Name=dyn_R_wing_3, bone2Name=dyn_R_wing_2, swing1LimitDegrees=7, swing2LimitDegrees=7, twistLimitDegrees=0}
 {bone1Name=dyn_R_wing_2, bone2Name=dyn_R_wing_1, swing1LimitDegrees=7, swing2LimitDegrees=7, twistLimitDegrees=0}
 {bone1Name=dyn_R_wing_1, bone2Name=R_wing, swing1LimitDegrees=5, swing2LimitDegrees=5, twistLimitDegrees=0}
 {bone1Name=dyn_L_wing_1, bone2Name=L_wing, swing1LimitDegrees=5, swing2LimitDegrees=5, twistLimitDegrees=0}
]
